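; Battle pass quest ability that broadcasts a CustomVerb.MaxAmmo message as the player maxes out ammo types.
; On activation it listens for filtered verb events on the avatar and runs Check Ammo.
; That compares inventory quantity with max stack size for five ammo types, adding each full type's tag uniquely to an array.
; When the array length exceeds TagCounter, it updates the counter and broadcasts the accumulated tags, preventing repeat reports.
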(event Ability|EventFailedToActivatePassiveAbility
  (params (ActorInfo :type Gameplay%20Ability%20Actor%20Info%20Structure))
  (|CallParentFunction ActorInfo))

(event Ability|EventOnEndAbility
  (params (bWasCancelled :type Boolean))
  (|CallParentFunction bWasCancelled))

(event Ability|EventActivateAbility
  (bind _returnvalue (Ability|GetAvatarActorfromActorInfo))
  (|CallParentFunction)
  (Utilities|IsValid _returnvalue
    (:Is%20Valid
      (EventProcessorSubsystem|StartListeningToFilteredEvent 0 _returnvalue (Variables|Default|GetVerb%20Filter) false (delegate Custom|Check%20Ammo)))))

(event Custom|Check%20Ammo
  (bind _tags (Variables|Default|GetTags))
  (bind _returnvalue (Utilities|Array|Length _tags))
  (bind _tags_2 (Variables|Default|GetTags))
  ;[1 1 1] Loop Through the players 5 ammo types and fire a custom if they have a Max count of ammo type (once per ammo type per match)
  (for (_array_element _array_index) (Variables|Default|GetAmmo%20Types)
    (if (>= (Inventory|GetItemQuantityonInventoryOwner (Fortnite|GetFortPlayerControllerfromActor (Ability|GetAvatarActorfromActorInfo)) _array_element) (Item|GetMaxStackSize _array_element 0))
      (switch int _array_index
        (:0
          (Utilities|Array|AddUnique _tags_2 "(TagName=\"supplies.ammo.bulletlight\")"))
        (:1
          (Utilities|Array|AddUnique _tags_2 "(TagName=\"supplies.ammo.shell\")"))
        (:2
          (Utilities|Array|AddUnique _tags_2 "(TagName=\"supplies.ammo.bulletmedium\")"))
        (:3
          (Utilities|Array|AddUnique _tags_2 "(TagName=\"supplies.ammo.bulletheavy\")"))
        (:4
          (Utilities|Array|AddUnique _tags_2 "(TagName=\"supplies.ammo.explosive\")")))))
  (if (> _returnvalue (Variables|Default|GetTagCounter))
    (Variables|Default|SetTagCounter _returnvalue)
    (bind _result (CustomVerb|AddTagsfromSubjectwithAdditionalTags (CustomVerb|MakeCustomVerbMessage (Fortnite|GetFortPlayerControllerfromActor (Variables|GAVars|Data|GetOwningPlayerPawn)) "(TagName=\"CustomVerb.MaxAmmo\")") "Ammo Type" 0 (GameplayTags|MakeGameplayTagContainerfromArray _tags_2)))
    (CustomVerb|BroadcastCustomVerbMessage _result)))
